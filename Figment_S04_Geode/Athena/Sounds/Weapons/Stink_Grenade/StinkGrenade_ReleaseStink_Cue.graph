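# Sound cue for the Stink Grenade's looping stink-release sound, in the ranged weapon sound class.
# Two looping layers, a main release wave and an "Inside" variant, each pass through a fixed pitch and volume modulator and their own smoke grenade attenuation.
# A mixer blends both, with the inside layer's input boosted, into a root modulator left at default settings.
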SoundCue /Figment_S04_Geode/Athena/Sounds/Weapons/Stink_Grenade/StinkGrenade_ReleaseStink_Cue
root=Modulator_0
nodes [
 {id=Attenuation_2, children [Modulator_2], properties {attenuationSettings=/Game/Sounds/Attenuation/Fort_Grenade_Smoke_Att}}
 {
  id=Attenuation_3
  children [Modulator_3]
  properties {attenuationSettings=/Game/Sounds/Attenuation/Fort_Grenade_Smoke_Inside_Att, attenuationOverrides {bAttenuate=false, bSpatialize=false}}
 }
 {id=Mixer_0, children [Attenuation_2, Attenuation_3], properties {inputVolume [1,1.2]}}
 {id=Modulator_0, children [Mixer_0]}
 {id=Modulator_2, children [WavePlayer_0], properties {pitchMax=0.9, pitchMin=0.9, volumeMax=0.7, volumeMin=0.7}}
 {id=Modulator_3, children [WavePlayer_3], properties {pitchMax=0.9, pitchMin=0.9, volumeMax=0.6, volumeMin=0.6}}
 {
  id=WavePlayer_0
  properties {bLooping=true, soundWaveAssetPtr=&/StinkGrenade_Release_02}
 }
 {
  id=WavePlayer_3
  properties {bLooping=true, soundWaveAssetPtr=&/StinkGrenade_Release_Inside_01}
 }
]
properties {
 attenuationSettings=&Fort_Grenade_Smoke_Att
 maxDistance=6800
 soundClassObject=&Fort_WeaponRanged
 volumeMultiplier=1
}
